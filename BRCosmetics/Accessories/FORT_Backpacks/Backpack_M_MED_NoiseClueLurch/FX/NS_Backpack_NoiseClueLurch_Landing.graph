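# Niagara effect for the NoiseClueLurch backpack, named for landing, that drops tumbling gold mesh pieces with sprite glints.
# A single Gold emitter combines a spawn rate with a small random burst, launching pieces from just below the system origin.
# Gravity, drag and age-blended rotation drive the motion, with bouncing collision gated by platform quality level.
# Each particle feeds mesh and sprite renderers, with a size curve briefly flashing the warm-tinted, camera-offset sprite.
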NiagaraSystem /BRCosmetics/Accessories/FORT_Backpacks/Backpack_M_MED_NoiseClueLurch/FX/NS_Backpack_NoiseClueLurch_Landing
dependencies {
 usedDataInterfaces [&NiagaraDataInterfacePlatformSet, &NiagaraDataInterfaceCurve]
 usedDynamicInputs []
 usedModules [&SystemState]
}
emitters {Gold {
  emitterUpdateScript [
   &EmitterState()
   &SpawnRate(SpawnRate=10, Use Spawn Probability=false)
   &SpawnBurst_Instantaneous(
    Spawn Count=&RandomRangeInt(Minimum=2, Maximum=6)
    Spawn Probability=0.5
    Use Spawn Probability=false
    UseLoopCountLimit=false
   )
  ]
  particleSpawnScript [
   &InitializeParticle(
    Color Mode=::NewEnumerator0
    Lifetime Max=1.5
    Lifetime Mode=::NewEnumerator1
    Mesh Scale Mode=::NewEnumerator3
    Mesh Uniform Scale=0.75
    UsePositionOffset=false
    Write Material Random=true
   )
   &AddVelocity(
    Velocity=&RandomRangeVector(Minimum=<-1 -1 -0.25>, Maximum=<1 -1 -0.1>)
    Velocity Speed Scale=&RandomRangeFloat(Minimum=50, Maximum=75)
   ) {module=AddVelocity001}
   &SystemLocation(Offset=&RandomRangeVector(Minimum=<-4 0 -7>, Maximum=<4 0 -5>))
   &InitialMeshOrientation(Mesh Orientation Mode=::NewEnumerator2, Orientation Coordinate Space=::Local, Rotation=<0 0 0.02>, Use Rotation Vector=true)
  ]
  particleUpdateScript [
   &ParticleState()
   &UpdateMeshOrientation(
    Rotation Rate=&RandomRangeFloat(Minimum=0.8, Maximum=1.2)
    Rotation Vector=&MakeVector(
     X=&Lerp_Float(A=0, B=&RandomRangeFloat(Minimum=0.8, Maximum=1.2), Alpha=$Particles.NormalizedAge)
     Y=&Lerp_Float(
      A=&RandomRangeFloat(Minimum=0.25, Maximum=0.5)
      B=&RandomRangeFloat(Minimum=0.4, Maximum=-0.4)
      Alpha=$Particles.NormalizedAge
     )
     Z=&Lerp_Float(A=0.1, B=0.2, Alpha=$Particles.NormalizedAge)
    )
   )
   &GravityForce(Gravity=<0 0 -500>)
   &Drag(Use Linear Drag=true, Use Rotational Drag=true)
   &Collision(
    Age Colliding Particles=false
    Collision Enabled=&Boolean_AndOperation(
     Bool1=&IsPlatformSetActive(New Platform Set=DataInterface() {propertyValues {Platforms {qualityLevelMask=-4}}})
     Bool2=&SetBoolByIntComparison(A=&RandomRangeInt(Minimum=0, Maximum=0), B=0, Comparison Type=::NewEnumerator4)
    )
    Control Roll On Collision=true
    Friction=0.45
    KillOnCollision=false
    Module.Control Rest State Via Surface Normal=false
    Module.EnableMaxCollisionCount=false
    Module.Randomize Collision Normal Vector=false
    Restitution=0.5
    ScaleRateOfRotation=false
   )
   &SolveForcesAndVelocity(Clamp Velocity=false, Limit Acceleration=false)
   &FN_Sprite_Attributes(
    Camera Offset:NiagaraFloat=&Add_Float(A=$Particles.CameraOffset, B=5)
    Color=<rgba 1 0.8481 0.5956 1>
    Create Camera Offset Attribute=true
    Create Color Attribute=true
    Create Position Attribute=true
    Create Sprite Rotation Attribute=true
    Create SpriteSize Attribute=true
    Module Output:ENiagara_AttributeSamplingApplyOutput=::NewEnumerator1
    Position:NiagaraPosition=&AddVectorToPosition(Position=$Particles.Position, Vector=<0 0 0>)
    SpriteRotation:NiagaraFloat=&RandomRangeFloat(Minimum=200, Maximum=360)
    SpriteSize=&Multiply_Vector2D_ByFloat(
     Float=&FloatFromCurve(
      FloatCurve=DataInterface() {propertyValues {Curve {
         defaultValue=0
         keys [
          {interpMode=RCIM_Cubic, tangentMode=RCTM_User, time=0.15, value=0}
          {interpMode=RCIM_Cubic, tangentMode=RCTM_User, time=0.3, value=1}
          {tangentMode=RCTM_User, time=0.5, value=0}
         ]
        }}}
      CurveIndex=&Multiply_Float(A=$Particles.NormalizedAge, B=&RandomRangeFloat(Minimum=0.8, Maximum=2))
     )
     Vector2D=&Vector2DFromFloat(Value=&RandomRangeFloat(Minimum=15, Maximum=20))
    )
   )
  ]
  rendererClasses [&NiagaraMeshRendererProperties, &NiagaraSpriteRendererProperties]
  renderers [
   {rendererClass=&NiagaraMeshRendererProperties, rendererIndex=0}
   {rendererClass=&NiagaraSpriteRendererProperties, rendererIndex=1}
  ]
 }}
summary {
 systemName=NS_Backpack_NoiseClueLurch_Landing
 emitters [{emitterName=Gold, rendererClasses [&NiagaraMeshRendererProperties, &NiagaraSpriteRendererProperties]}]
 userVariables [User.bIsFrontEndPreview {type=NiagaraBool}]
}
